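# Green spawn effect for the CarrotCake player emote, attached to two skeletal mesh bones or sockets.
# Flash_Flare and CapsoleGlow fire staggered bursts offset along each point's local Y, the glow passing its rotated axis to the material and adding lights.
# EmbersRight and delayed EmbersLeft sweep spawn points outward over each loop.
# Embers inherit fading bone movement, slowed by drag and stirred by curl noise.
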NiagaraSystem /BRCosmetics/Animation/Game/MainPlayer/Emotes/CarrotCake/FX/NS_Emote_CarrotCake_Spawn
dependencies {
 usedDataInterfaces [
  &NiagaraDataInterfaceCurve
  &NiagaraDataInterfaceVector2DCurve
  &NiagaraDataInterfaceArrayFloat3
  &NiagaraDataInterfaceColorCurve
  &NiagaraDataInterfaceSkeletalMesh
 ]
 usedDynamicInputs []
 usedModules [
  &self:SystemSpawnScript.SystemScriptSource.SystemScriptGraph.NiagaraNodeAssignment_0.SetVariables_#1
  &SystemState
 ]
}
emitters {
 CapsoleGlow {
  emitterSpawnScript [/SetVariables(Emitter.CapsuleGlowDirection:Vector3f=<0 1 0>) {isSetParameters=true}]
  emitterUpdateScript [
   &EmitterState(Inactive Response=::NewEnumerator1, Life Cycle Mode=::NewEnumerator1, Loop Behavior=::NewEnumerator1, Loop Duration=2, UseLoopDelay=false)
   &SpawnBurst_Instantaneous(Use Spawn Probability=false, UseLoopCountLimit=false)
   &SpawnBurst_Instantaneous(Spawn Count=1, Spawn Time=0.25, Use Spawn Probability=false) {module=SpawnBurst_Instantaneous001}
  ]
  particleSpawnScript [
   &InitializeParticle(
    Color=<rgba 0 2.5 0.7083 1>
    Lifetime=0.3
    Sprite Size Mode=::NewEnumerator3
    Uniform Sprite Size=30
    UsePositionOffset=false
    Write Material Random=true
   )
   &FN_SetSpawnGroup(Number of Spawn Groups=2)
  ]
  particleUpdateScript [
   &ParticleState()
   &SkeletalMeshLocation(
    Bone / Socket Index=&SpawnGroup()
    Bone Sampling Mode=::NewEnumerator4
    MaskSpawnGroup=false
    Orientation Sampling=::NewEnumerator1
    Sampled Position Offset=&MakeVector(X=0, Y=&MaskFloatBySpawnGroup(Masked Float=-5, Passthrough Float=5), Z=0)
    Skeletal Mesh=$System.Skeletal Mesh
    Use Sampled Mesh Attribute as Ribbon ID=false
   )
   &DynamicMaterialParameters(
    Index 0 Param 1=&MakeFloatFromVector(
     Vector=&RotateVectorByQuaternion(Vector To Rotate=$Emitter.CapsuleGlowDirection, Quaternion=$Particles.SkeletalMeshLocation.SampledRotation)
     Channel=::NewEnumerator0
    )
    Index 0 Param 2=&MakeFloatFromVector(
     Vector=&RotateVectorByQuaternion(Vector To Rotate=$Emitter.CapsuleGlowDirection, Quaternion=$Particles.SkeletalMeshLocation.SampledRotation)
     Channel=::NewEnumerator1
    )
    Index 0 Param 3=&MakeFloatFromVector(
     Vector=&RotateVectorByQuaternion(Vector To Rotate=$Emitter.CapsuleGlowDirection, Quaternion=$Particles.SkeletalMeshLocation.SampledRotation)
     Channel=::NewEnumerator2
    )
    Index 0 Param 4=50
    Module.Param0WriteEnabled=true
    Module.Param1WriteEnabled=true
    Module.Param2WriteEnabled=true
    Module.Param3WriteEnabled=true
   )
   /SetVariables(Particles.Position=&AddVectorToPosition(
     Position=$Particles.Position
     Vector=&Multiply_VectorByFloat(
      Vector=&RotateVectorByQuaternion(
       Vector To Rotate=&SelectVectorFromArray(
        Direct Array Index=&SpawnGroup()
        Vector Selection Array=DataInterface() {propertyValues {FloatData [<0 -1 0>, <0 1 0>]}}
       )
       Quaternion=$Particles.SkeletalMeshLocation.SampledRotation
      )
      Float=10
     )
    )) {isSetParameters=true}
   &ScaleColor(
    Scale Alpha=&FloatFromCurve(FloatCurve=DataInterface() {propertyValues {Curve {keys [
         {arriveTangent=7.0246, interpMode=RCIM_Cubic, leaveTangent=7.0246, tangentMode=RCTM_User, time=0, value=0}
         {interpMode=RCIM_Cubic, tangentMode=RCTM_User, time=0.3, value=1}
         {tangentMode=RCTM_User, time=1, value=0}
        ]}}})
    ScaleA=true
    ScaleRGB=true
   ) {module=ScaleColor001}
   &FN_Light_Attributes(
    Color=$Particles.Color
    Create Color Attribute=true
    Create Radius Attribute=true
    Enable Particle Light Enable=true
    Light Enabled=true
    Radius=100
    Scale Color=&FloatFromCurve(
     FloatCurve=DataInterface() {propertyValues {Curve {keys [
         {arriveTangent=4.2857, interpMode=RCIM_Cubic, leaveTangent=4.2857, tangentMode=RCTM_User, time=0, value=0}
         {interpMode=RCIM_Cubic, tangentMode=RCTM_User, time=0.5, value=1}
         {tangentMode=RCTM_User, time=1, value=0}
        ]}}}
     Scale Curve=1.5
    )
   )
  ]
  rendererClasses [&NiagaraSpriteRendererProperties, &NiagaraLightRendererProperties]
  renderers [
   {rendererClass=&NiagaraSpriteRendererProperties}
   {rendererClass=&NiagaraLightRendererProperties, rendererIndex=1}
  ]
 }
 EmbersLeft {
  emitterUpdateScript [
   &EmitterState(Life Cycle Mode=::NewEnumerator1, Loop Behavior=::NewEnumerator1, Loop Delay=0.15, Loop Duration=0.3, UseLoopDelay=true)
   &SpawnRate(SpawnRate=250, Use Spawn Probability=true) {module=SpawnRate001}
  ]
  particleSpawnScript [
   /SetVariables(Particles.PerParticleRandom:NiagaraFloat=&RandomRangeFloat(Minimum=0, Maximum=1)) {isSetParameters=true, module=SetVariables_#2}
   &FN_SetSpawnGroup(Number of Spawn Groups=2)
   &InitializeParticle(
    Color=<rgba 0 1 0.0964 1>
    Lifetime=&Lerp_Float(A=0.4, B=0.6, Alpha=$Particles.PerParticleRandom)
    Sprite Size=&Multiply_Vector2D(
     A=&RandomRangeVector2D(Minimum=<0.2 0.2>, Maximum=<0.3 0.5>)
     B=&MakeVector2D(
      X=&FloatFromCurve(
       FloatCurve=DataInterface() {propertyValues {Curve {keys [<at 0 1>,<at 0.8 2 RCIM_Constant>,<at 0.95 3 RCIM_Constant>]}}}
       CurveIndex=$Particles.PerParticleRandom
      )
      Y=&FloatFromCurve(
       FloatCurve=DataInterface() {propertyValues {Curve {keys [<at 0 1>,<at 0.8 2 RCIM_Constant>,<at 0.95 3>]}}}
       CurveIndex=$Particles.PerParticleRandom
      )
     )
    )
    Sprite Size Mode=::NewEnumerator4
    UsePositionOffset=false
    Write Material Random=true
   )
   &SkeletalMeshLocation(
    Bone / Socket Index=1
    Bone Sampling Mode=::NewEnumerator4
    MaskSpawnGroup=false
    Orientation Sampling=::NewEnumerator1
    Skeletal Mesh=$System.Skeletal Mesh
   )
   /SetVariables(Particles.Position=&AddVectorToPosition(
     Position=$Particles.Position
     Vector=&Multiply_VectorByFloat(
      Vector=&RotateVectorByQuaternion(Vector To Rotate=<0 1 0>, Quaternion=$Particles.SkeletalMeshLocation.SampledRotation)
      Float=&FloatFromCurve(
       FloatCurve=DataInterface() {propertyValues {Curve {keys [{tangentMode=RCTM_User, time=0, value=0}, {tangentMode=RCTM_User, time=1, value=1}]}}}
       CurveIndex=$Emitter.NormalizedLoopAge
       Scale Curve=30
      )
     )
    )) {isSetParameters=true}
   &AddVelocity(
    Cone Angle=90
    Cone Axis=&RotateVectorByQuaternion(Vector To Rotate=<0 1 0>, Quaternion=<-0.39 -0.01 0.18 -0.9>)
    Module.UseConeFalloff=true
    Module.UseVelDistribution=false
    Override Seed=false
    Velocity Mode=::NewEnumerator2
    Velocity Speed=&Multiply_Float(A=&Lerp_Float(A=100, B=200, Alpha=$Particles.PerParticleRandom), B=1)
   )
  ]
  particleUpdateScript [
   &ParticleState()
   &SkeletalMeshLocation(
    Bone / Socket Index=1
    Bone Sampling Mode=::NewEnumerator4
    Initialize Particle Position To Sampled Position=false
    MaskSpawnGroup=false
    Orientation Sampling=::NewEnumerator1
    Position Sampling=::NewEnumerator1
    Skeletal Mesh=$System.Skeletal Mesh
    Source Movement Scale:Vector3f=&VectorFromFloat(Value=&FloatFromCurve(FloatCurve=DataInterface() {propertyValues {Curve {keys [
          {arriveTangent=-2.587, interpMode=RCIM_Cubic, leaveTangent=-2.587, tangentMode=RCTM_User, time=0, value=1}
          {interpMode=RCIM_Cubic, tangentMode=RCTM_User, time=1, value=0}
         ]}}}))
   ) {module=SkeletalMeshLocation001}
   &Drag(
    Drag=&FloatFromCurve(
     FloatCurve=DataInterface() {propertyValues {Curve {keys [
         {interpMode=RCIM_Cubic, tangentMode=RCTM_User, time=0, value=0.4}
         {arriveTangent=0.9703, interpMode=RCIM_Cubic, leaveTangent=0.9703, tangentMode=RCTM_User, time=1, value=1}
        ]}}}
     Scale Curve=&RandomRangeFloat(Minimum=20, Maximum=30)
    )
    Use Linear Drag=true
    Use Rotational Drag=false
   )
   &CurlNoiseForce(Noise Strength=500, Pan Noise=true, Pan Noise Field=<0 0 0.1>) {module=CurlNoiseForce001}
   &SolveForcesAndVelocity(Clamp Velocity=false, Limit Acceleration=false)
   &ScaleColor(
    Scale RGB=&MakeVectorFromLinearColorRGB(Color=&ColorFromCurve(ColorCurve=DataInterface() {propertyValues {
        AlphaCurve {keys [<at 0 1 RCIM_Cubic>]}
        BlueCurve {keys [<at 0.0039 5>,<at 1 0.576 RCIM_Cubic>]}
        GreenCurve {keys [<at 0.0039 5>,<at 1 0.576 RCIM_Cubic>]}
        RedCurve {keys [<at 0.0039 5>,<at 1 0.576 RCIM_Cubic>]}
       }}))
    ScaleA=true
    ScaleRGB=true
   )
   &ScaleSpriteSize(
    Uniform Curve Scale=&MakeFloatFromVector(Vector=$Engine.Owner.Scale, Channel=::NewEnumerator0)
    Uniform Curve Sprite Scale=DataInterface() {propertyValues {Curve {keys [
        {tangentMode=RCTM_User, time=0, value=1}
        {arriveTangent=-0.0405, interpMode=RCIM_Cubic, leaveTangent=-0.0405, tangentMode=RCTM_User, time=0.7009, value=1}
        {arriveTangent=-8.0089, interpMode=RCIM_Cubic, leaveTangent=-8.0089, tangentMode=RCTM_User, time=1, value=0}
       ]}}}
   )
   &ScaleSpriteSizeBySpeed(
    Max Scale Factor=<1.666 3>
    Min Scale Factor=<1.333 1.333>
    Sample Scale Factor By Curve=true
    Scale Factor Curve=DataInterface() {propertyValues {Curve {keys [<at 0 0>, {arriveTangent=1.4989, interpMode=RCIM_Cubic, tangentMode=RCTM_Break, time=1, value=1}]}}}
    Velocity Threshold=250
   )
  ]
  rendererClasses [&NiagaraSpriteRendererProperties]
  renderers [{rendererClass=&NiagaraSpriteRendererProperties}]
 }
 EmbersRight {
  emitterUpdateScript [
   &EmitterState(Life Cycle Mode=::NewEnumerator1, Loop Behavior=::NewEnumerator1, Loop Duration=0.3, UseLoopDelay=true)
   &SpawnRate(SpawnRate=250, Use Spawn Probability=true) {module=SpawnRate001}
  ]
  particleSpawnScript [
   /SetVariables(Particles.PerParticleRandom:NiagaraFloat=&RandomRangeFloat(Minimum=0, Maximum=1)) {isSetParameters=true, module=SetVariables_#2}
   &FN_SetSpawnGroup(Number of Spawn Groups=2)
   &InitializeParticle(
    Color=<rgba 0 1 0.0964 1>
    Lifetime=&Lerp_Float(A=0.4, B=0.6, Alpha=$Particles.PerParticleRandom)
    Sprite Size=&Multiply_Vector2D(
     A=&RandomRangeVector2D(Minimum=<0.2 0.2>, Maximum=<0.3 0.5>)
     B=&MakeVector2D(
      X=&FloatFromCurve(
       FloatCurve=DataInterface() {propertyValues {Curve {keys [<at 0 1>,<at 0.8 2 RCIM_Constant>,<at 0.95 3 RCIM_Constant>]}}}
       CurveIndex=$Particles.PerParticleRandom
      )
      Y=&FloatFromCurve(
       FloatCurve=DataInterface() {propertyValues {Curve {keys [<at 0 1>,<at 0.8 2 RCIM_Constant>,<at 0.95 3>]}}}
       CurveIndex=$Particles.PerParticleRandom
      )
     )
    )
    Sprite Size Mode=::NewEnumerator4
    UsePositionOffset=false
    Write Material Random=true
   )
   &SkeletalMeshLocation(Bone Sampling Mode=::NewEnumerator4, MaskSpawnGroup=false, Orientation Sampling=::NewEnumerator1, Skeletal Mesh=$System.Skeletal Mesh)
   /SetVariables(Particles.Position=&AddVectorToPosition(
     Position=$Particles.Position
     Vector=&Multiply_VectorByFloat(
      Vector=&RotateVectorByQuaternion(Vector To Rotate=<0 -1 0>, Quaternion=$Particles.SkeletalMeshLocation.SampledRotation)
      Float=&FloatFromCurve(
       FloatCurve=DataInterface() {propertyValues {Curve {keys [{tangentMode=RCTM_User, time=0, value=0}, {tangentMode=RCTM_User, time=1, value=1}]}}}
       CurveIndex=$Emitter.NormalizedLoopAge
       Scale Curve=30
      )
     )
    )) {isSetParameters=true}
   &AddVelocity(
    Cone Angle=60
    Cone Axis=&RotateVectorByQuaternion(Vector To Rotate=<0 -1 0>, Quaternion=<0.9 0.11 0.03 -0.42>)
    Module.UseConeFalloff=true
    Module.UseVelDistribution=false
    Override Seed=false
    Velocity Mode=::NewEnumerator2
    Velocity Speed=&Multiply_Float(A=&Lerp_Float(A=100, B=200, Alpha=$Particles.PerParticleRandom), B=1)
   )
   /SetVariables(Particles.Quat:Quat4f=$Particles.SkeletalMeshLocation.SampledRotation) {isSetParameters=true}
  ]
  particleUpdateScript [
   &ParticleState()
   &SkeletalMeshLocation(
    Bone Sampling Mode=::NewEnumerator4
    Initialize Particle Position To Sampled Position=false
    MaskSpawnGroup=false
    Orientation Sampling=::NewEnumerator1
    Position Sampling=::NewEnumerator1
    Skeletal Mesh=$System.Skeletal Mesh
    Source Movement Scale:Vector3f=&VectorFromFloat(Value=&FloatFromCurve(FloatCurve=DataInterface() {propertyValues {Curve {keys [
          {arriveTangent=-1.9895, interpMode=RCIM_Cubic, leaveTangent=-1.9895, tangentMode=RCTM_User, time=0, value=1}
          {interpMode=RCIM_Cubic, tangentMode=RCTM_User, time=1, value=0}
         ]}}}))
   ) {module=SkeletalMeshLocation001}
   &Drag(
    Drag=&FloatFromCurve(
     FloatCurve=DataInterface() {propertyValues {Curve {keys [
         {interpMode=RCIM_Cubic, tangentMode=RCTM_User, time=0, value=0.4}
         {arriveTangent=0.9703, interpMode=RCIM_Cubic, leaveTangent=0.9703, tangentMode=RCTM_User, time=1, value=1}
        ]}}}
     Scale Curve=&RandomRangeFloat(Minimum=20, Maximum=30)
    )
    Use Linear Drag=true
    Use Rotational Drag=false
   )
   &CurlNoiseForce(Noise Strength=500, Pan Noise=true, Pan Noise Field=<0 0 0.1>) {module=CurlNoiseForce001}
   &SolveForcesAndVelocity(Clamp Velocity=false, Limit Acceleration=false)
   &ScaleColor(
    Scale RGB=&MakeVectorFromLinearColorRGB(Color=&ColorFromCurve(ColorCurve=DataInterface() {propertyValues {
        AlphaCurve {keys [<at 0 1 RCIM_Cubic>]}
        BlueCurve {keys [<at 0.0039 5>,<at 1 0.576 RCIM_Cubic>]}
        GreenCurve {keys [<at 0.0039 5>,<at 1 0.576 RCIM_Cubic>]}
        RedCurve {keys [<at 0.0039 5>,<at 1 0.576 RCIM_Cubic>]}
       }}))
    ScaleA=true
    ScaleRGB=true
   )
   &ScaleSpriteSize(
    Uniform Curve Scale=&MakeFloatFromVector(Vector=$Engine.Owner.Scale, Channel=::NewEnumerator0)
    Uniform Curve Sprite Scale=DataInterface() {propertyValues {Curve {keys [
        {tangentMode=RCTM_User, time=0, value=1}
        {arriveTangent=-0.0405, interpMode=RCIM_Cubic, leaveTangent=-0.0405, tangentMode=RCTM_User, time=0.7009, value=1}
        {arriveTangent=-8.0089, interpMode=RCIM_Cubic, leaveTangent=-8.0089, tangentMode=RCTM_User, time=1, value=0}
       ]}}}
   )
   &ScaleSpriteSizeBySpeed(
    Max Scale Factor=<1.666 3>
    Min Scale Factor=<1.333 1.333>
    Sample Scale Factor By Curve=true
    Scale Factor Curve=DataInterface() {propertyValues {Curve {keys [<at 0 0>, {arriveTangent=1.4989, interpMode=RCIM_Cubic, tangentMode=RCTM_Break, time=1, value=1}]}}}
    Velocity Threshold=250
   )
  ]
  rendererClasses [&NiagaraSpriteRendererProperties]
  renderers [{rendererClass=&NiagaraSpriteRendererProperties}]
 }
 Flash_Flare {
  emitterUpdateScript [
   &EmitterState(
    Inactive Response=::NewEnumerator1
    Life Cycle Mode=::NewEnumerator1
    Loop Behavior=::NewEnumerator1
    Loop Duration Mode=::NewEnumerator1
    UseLoopDelay=true
   )
   &SpawnBurst_Instantaneous(Spawn Count=1, Use Spawn Probability=false)
   &SpawnBurst_Instantaneous(Spawn Count=1, Spawn Time=0.2, Use Spawn Probability=false) {module=SpawnBurst_Instantaneous001}
  ]
  particleSpawnScript [
   &InitializeParticle(
    Color=<rgba 0 1 0.0667 1>
    Lifetime=0.2
    Sprite Size=<40 60>
    Sprite Size Mode=::NewEnumerator4
    UsePositionOffset=false
    Write Material Random=true
   )
   &FN_SetSpawnGroup(Number of Spawn Groups=2)
   &DynamicMaterialParameters(
    Index 0 Param 1=0.5
    Module.Param0WriteEnabled=true
    Module.Param1WriteEnabled=true
    Module.Param2WriteEnabled=true
    Module.Param3WriteEnabled=true
   )
   &CameraOffset(Camera Offset Amount=5) {module=CameraOffset001}
  ]
  particleUpdateScript [
   &ParticleState()
   &SkeletalMeshLocation(
    Bone / Socket Index=&SpawnGroup()
    Bone Sampling Mode=::NewEnumerator4
    MaskSpawnGroup=false
    Orientation Sampling=::NewEnumerator1
    Skeletal Mesh=$System.Skeletal Mesh
   )
   &ScaleSpriteSize(
    Non-Uniform Curve Sprite Scale=DataInterface() {propertyValues {
      XCurve {keys [
        {arriveTangent=9.0725, interpMode=RCIM_Cubic, leaveTangent=9.0725, tangentMode=RCTM_User, time=0, value=0}
        {arriveTangent=-2.5249, interpMode=RCIM_Cubic, leaveTangent=-2.5249, tangentMode=RCTM_User, time=0.3, value=2}
        {arriveTangent=-0.0405, interpMode=RCIM_Cubic, leaveTangent=-0.0405, tangentMode=RCTM_User, time=1, value=1}
       ]}
      YCurve {keys [
        {interpMode=RCIM_Cubic, tangentMode=RCTM_User, time=0, value=0}
        {interpMode=RCIM_Cubic, tangentMode=RCTM_User, time=0.5, value=1}
        {tangentMode=RCTM_User, time=1, value=0.6}
       ]}
     }}
    Scale Sprite Size Mode=::NewEnumerator3
   )
   &ScaleColor(
    Scale Alpha=&FloatFromCurve(FloatCurve=DataInterface() {propertyValues {Curve {keys [
         {arriveTangent=9.0725, interpMode=RCIM_Cubic, leaveTangent=9.0725, tangentMode=RCTM_User, time=0, value=0}
         {arriveTangent=0.0145, interpMode=RCIM_Cubic, leaveTangent=0.0145, tangentMode=RCTM_User, time=0.1408, value=1}
         {arriveTangent=-3.2545, interpMode=RCIM_Cubic, leaveTangent=-3.2545, tangentMode=RCTM_User, time=1, value=0}
        ]}}})
    Scale RGB=&VectorFromFloat(Value=&FloatFromCurve(
      FloatCurve=DataInterface() {propertyValues {Curve {keys [
          {arriveTangent=9.0725, interpMode=RCIM_Cubic, leaveTangent=9.0725, tangentMode=RCTM_User, time=0, value=0}
          {arriveTangent=-2.5249, interpMode=RCIM_Cubic, leaveTangent=-2.5249, tangentMode=RCTM_User, time=0.1408, value=1}
          {arriveTangent=-0.0405, interpMode=RCIM_Cubic, leaveTangent=-0.0405, tangentMode=RCTM_User, time=1, value=0}
         ]}}}
      Scale Curve=5
     ))
    ScaleA=true
    ScaleRGB=true
   )
   /SetVariables(Particles.Position=&AddVectorToPosition(
     Position=$Particles.Position
     Vector=&Multiply_VectorByFloat(
      Vector=&RotateVectorByQuaternion(
       Vector To Rotate=&SelectVectorFromArray(
        Direct Array Index=&SpawnGroup()
        Vector Selection Array=DataInterface() {propertyValues {FloatData [<0 -1 0>, <0 1 0>]}}
       )
       Quaternion=$Particles.SkeletalMeshLocation.SampledRotation
      )
      Float=&FloatFromCurve(
       FloatCurve=DataInterface() {propertyValues {Curve {keys [{arriveTangent=1.7577, interpMode=RCIM_Cubic, leaveTangent=1.7577, tangentMode=RCTM_User, time=0, value=0}, <at 1 1 RCIM_Cubic>]}}}
       Scale Curve=31
      )
     )
    )) {isSetParameters=true}
  ]
  rendererClasses [&NiagaraSpriteRendererProperties]
  renderers [{rendererClass=&NiagaraSpriteRendererProperties}]
 }
}
summary {
 systemName=NS_Emote_CarrotCake_Spawn
 emitters [
  {emitterName=Flash_Flare, rendererClasses [&NiagaraSpriteRendererProperties]}
  {emitterName=EmbersRight, rendererClasses [&NiagaraSpriteRendererProperties]}
  {emitterName=EmbersLeft, rendererClasses [&NiagaraSpriteRendererProperties]}
  {emitterName=CapsoleGlow, rendererClasses [&NiagaraSpriteRendererProperties, &NiagaraLightRendererProperties]}
 ]
 userVariables [
  User.bIsFrontEnd {type=NiagaraBool}
  User.bIsFrontEndPreview {type=NiagaraBool}
  User.RibbonIDOffset {type=NiagaraInt32}
 ]
}
