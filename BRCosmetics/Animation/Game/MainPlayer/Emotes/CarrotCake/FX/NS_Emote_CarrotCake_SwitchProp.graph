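# Niagara burst effect for the CarrotCake emote, by its name covering a prop switch with a short green flash.
# All five emitters spawn at a skeletal mesh bone chosen by User.BoneIndex.
# A rising flare, an air glow with a particle light, curl-noise embers and flipbook smoke surround an upward mesh shockwave.
# The wave animates dynamic material parameters and dims its alpha by time of day.
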NiagaraSystem /BRCosmetics/Animation/Game/MainPlayer/Emotes/CarrotCake/FX/NS_Emote_CarrotCake_SwitchProp
dependencies {
 usedDataInterfaces [
  &NiagaraDataInterfaceCurve
  &NiagaraDataInterfaceVector2DCurve
  &NiagaraDataInterfaceColorCurve
  &NiagaraDataInterfaceSkeletalMesh
 ]
 usedDynamicInputs []
 usedModules [
  &self:SystemSpawnScript.SystemScriptSource.SystemScriptGraph.NiagaraNodeAssignment_0.SetVariables_#1
  &SystemState
 ]
}
emitters {
 AirGlow {
  emitterUpdateScript [
   &EmitterState(Life Cycle Mode=::NewEnumerator1, Loop Behavior=::NewEnumerator1, Loop Delay=0.5, Loop Duration=3, UseLoopDelay=false)
   &SpawnBurst_Instantaneous(Spawn Count=1, Use Spawn Probability=false)
  ]
  particleSpawnScript [
   &InitializeParticle(
    Color=<rgba 0 2 0.3223 0.1>
    Lifetime=0.5
    Sprite Size Mode=::NewEnumerator3
    Uniform Sprite Size=30
    UsePositionOffset=false
    Write Material Random=true
   ) {module=InitializeParticle001}
   &DynamicMaterialParameters(
    Index 0 Param 1=2
    Module.Param0WriteEnabled=true
    Module.Param1WriteEnabled=true
    Module.Param2WriteEnabled=true
    Module.Param3WriteEnabled=true
   )
   &CameraOffset(Camera Offset Amount=10)
   &SkeletalMeshLocation(Bone / Socket Index=$User.BoneIndex, Bone Sampling Mode=::NewEnumerator4, MaskSpawnGroup=false, Skeletal Mesh=$System.Skeletal Mesh)
  ]
  particleUpdateScript [
   &ParticleState()
   &ScaleSpriteSize(Scale Factor=&Vector2DFromFloat(Value=&FloatFromCurve(FloatCurve=DataInterface() {propertyValues {Curve {keys [<at 0 1 RCIM_Cubic>]}}})))
   &ScaleColor(
    Scale Alpha=&FloatFromCurve(FloatCurve=DataInterface() {propertyValues {Curve {keys [
         {arriveTangent=9.0725, interpMode=RCIM_Cubic, leaveTangent=9.0725, tangentMode=RCTM_User, time=0, value=0}
         {arriveTangent=-2.5249, interpMode=RCIM_Cubic, leaveTangent=-2.5249, tangentMode=RCTM_User, time=0.1408, value=0.9847}
         {arriveTangent=-0.0405, interpMode=RCIM_Cubic, leaveTangent=-0.0405, tangentMode=RCTM_User, time=1, value=0}
        ]}}})
    ScaleA=true
    ScaleRGB=true
   ) {module=ScaleColor001}
   &FN_Light_Attributes(
    Color=$Particles.Color
    Create Color Attribute=true
    Create Radius Attribute=true
    Enable Particle Light Enable=true
    Light Enabled=true
    Radius=100
    Scale Color=&FloatFromCurve(FloatCurve=DataInterface() {propertyValues {Curve {keys [
         {arriveTangent=4.2857, interpMode=RCIM_Cubic, leaveTangent=4.2857, tangentMode=RCTM_User, time=0, value=0}
         {interpMode=RCIM_Cubic, tangentMode=RCTM_User, time=0.5, value=1}
         {tangentMode=RCTM_User, time=1, value=0}
        ]}}})
   )
  ]
  rendererClasses [&NiagaraSpriteRendererProperties, &NiagaraLightRendererProperties]
  renderers [
   {rendererClass=&NiagaraSpriteRendererProperties}
   {rendererClass=&NiagaraLightRendererProperties, rendererIndex=1}
  ]
 }
 Embers {
  emitterUpdateScript [
   &EmitterState(Life Cycle Mode=::NewEnumerator1, Loop Behavior=::NewEnumerator1, Loop Duration=0.2, UseLoopDelay=false)
   &SpawnBurst_Instantaneous(Spawn Count=75, Use Spawn Probability=false)
   &SpawnRate(SpawnRate=150, Use Spawn Probability=true) {module=SpawnRate001}
  ]
  particleSpawnScript [
   /SetVariables(Particles.PerParticleRandom:NiagaraFloat=&RandomRangeFloat(Minimum=0, Maximum=1)) {isSetParameters=true}
   &InitializeParticle(
    Color Maximum=<rgba 0.3423 1 0 1>
    Color Minimum=<rgba 0 3 0.4656 1>
    Color Mode=::NewEnumerator2
    Lifetime=&Lerp_Float(A=0.3, B=0.5, Alpha=$Particles.PerParticleRandom)
    Sprite Size=&Multiply_Vector2D(
     A=&RandomRangeVector2D(Minimum=<0.2 0.2>, Maximum=<0.2 0.5>)
     B=&MakeVector2D(
      X=&FloatFromCurve(
       FloatCurve=DataInterface() {propertyValues {Curve {keys [<at 0 1>,<at 0.8 2 RCIM_Constant>,<at 0.95 3 RCIM_Constant>]}}}
       CurveIndex=$Particles.PerParticleRandom
      )
      Y=&FloatFromCurve(
       FloatCurve=DataInterface() {propertyValues {Curve {keys [<at 0 1>,<at 0.8 2 RCIM_Constant>,<at 0.95 3>]}}}
       CurveIndex=$Particles.PerParticleRandom
      )
     )
    )
    Sprite Size Mode=::NewEnumerator4
    UsePositionOffset=false
    Write Material Random=true
   )
   &SkeletalMeshLocation(Bone / Socket Index=$User.BoneIndex, Bone Sampling Mode=::NewEnumerator4, MaskSpawnGroup=false, Skeletal Mesh=$System.Skeletal Mesh)
   &AddVelocity(
    Module.Use Falloff Distance=false
    Override Seed=false
    Velocity Mode=::NewEnumerator1
    Velocity Origin=$Particles.Position
    Velocity Speed=&RandomRangeFloat(Minimum=15, Maximum=30)
   ) {module=AddVelocity001}
   &AddVelocity(
    Cone Angle=30
    Cone Axis=<0 0 1>
    Module.UseConeFalloff=false
    Module.UseVelDistribution=false
    Override Seed=false
    Rotation Coordinate Space=::World
    Velocity Mode=::NewEnumerator2
    Velocity Speed=&RandomRangeFloat(Minimum=50, Maximum=100)
   )
   &ShapeLocation(MaskSpawnGroup=false, Override Seed=false, Sphere Radius=2, UseOwnerScale=true)
  ]
  particleUpdateScript [
   &ParticleState()
   &Drag(
    Drag=&FloatFromCurve(
     FloatCurve=DataInterface() {propertyValues {Curve {keys [
         {interpMode=RCIM_Cubic, tangentMode=RCTM_User, time=0, value=0.6}
         {arriveTangent=0.9703, interpMode=RCIM_Cubic, leaveTangent=0.9703, tangentMode=RCTM_User, time=1, value=1}
        ]}}}
     Scale Curve=&RandomRangeFloat(Minimum=8, Maximum=12)
    )
    Use Linear Drag=true
    Use Rotational Drag=false
   )
   &CurlNoiseForce(Noise Frequency=100, Noise Strength=500, Pan Noise=true, Pan Noise Field=<0 0 -0.1>) {module=CurlNoiseForce001}
   &SolveForcesAndVelocity(Clamp Velocity=false, Limit Acceleration=false)
   &ScaleColor(
    Scale RGB=&MakeVectorFromLinearColorRGB(Color=&ColorFromCurve(ColorCurve=DataInterface() {propertyValues {
        AlphaCurve {keys [<at 0 1 RCIM_Cubic>]}
        BlueCurve {keys [<at 0.0039 5>,<at 1 0.576 RCIM_Cubic>]}
        GreenCurve {keys [<at 0.0039 5>,<at 1 0.576 RCIM_Cubic>]}
        RedCurve {keys [<at 0.0039 5>,<at 1 0.576 RCIM_Cubic>]}
       }}))
    ScaleA=true
    ScaleRGB=true
   )
   &ScaleSpriteSize(Uniform Curve Sprite Scale=DataInterface() {propertyValues {Curve {keys [
        {tangentMode=RCTM_User, time=0, value=1}
        {arriveTangent=-0.0405, interpMode=RCIM_Cubic, leaveTangent=-0.0405, tangentMode=RCTM_User, time=0.7009, value=1}
        {arriveTangent=-8.0089, interpMode=RCIM_Cubic, leaveTangent=-8.0089, tangentMode=RCTM_User, time=1, value=0}
       ]}}})
   &ScaleSpriteSizeBySpeed(
    Max Scale Factor=<1 3>
    Sample Scale Factor By Curve=true
    Scale Factor Curve=DataInterface() {propertyValues {Curve {keys [<at 0 0>, {arriveTangent=1.4989, interpMode=RCIM_Cubic, tangentMode=RCTM_Break, time=1, value=1}]}}}
    Velocity Threshold=150
   )
  ]
  rendererClasses [&NiagaraSpriteRendererProperties]
  renderers [{rendererClass=&NiagaraSpriteRendererProperties}]
 }
 Flash_Flare {
  emitterUpdateScript [
   &EmitterState(Life Cycle Mode=::NewEnumerator1, Loop Behavior=::NewEnumerator1, UseLoopDelay=false)
   &SpawnBurst_Instantaneous(Spawn Count=1, Use Spawn Probability=false)
  ]
  particleSpawnScript [
   &InitializeParticle(
    Color=<rgba 0 1 0.0381 1>
    Lifetime=0.3
    Sprite Size=<100 100>
    Sprite Size Mode=::NewEnumerator4
    UsePositionOffset=false
    Write Material Random=true
   )
   &DynamicMaterialParameters(
    Index 0 Param 1=0.5
    Module.Param0WriteEnabled=true
    Module.Param1WriteEnabled=true
    Module.Param2WriteEnabled=true
    Module.Param3WriteEnabled=true
   )
   &CameraOffset(Camera Offset Amount=10) {module=CameraOffset001}
   &SkeletalMeshLocation(Bone / Socket Index=$User.BoneIndex, Bone Sampling Mode=::NewEnumerator4, MaskSpawnGroup=false, Skeletal Mesh=$System.Skeletal Mesh)
  ]
  particleUpdateScript [
   &ParticleState()
   &ScaleSpriteSize(
    Non-Uniform Curve Sprite Scale=DataInterface() {propertyValues {
      XCurve {keys [
        {interpMode=RCIM_Cubic, tangentMode=RCTM_User, time=0, value=0}
        {interpMode=RCIM_Cubic, tangentMode=RCTM_User, time=0.5, value=1}
        {tangentMode=RCTM_User, time=1, value=0}
       ]}
      YCurve {keys [
        {interpMode=RCIM_Cubic, tangentMode=RCTM_User, time=0, value=0}
        {interpMode=RCIM_Cubic, tangentMode=RCTM_User, time=0.5, value=1}
        {tangentMode=RCTM_User, time=1, value=0}
       ]}
     }}
    Scale Sprite Size Mode=::NewEnumerator3
   )
   &ScaleColor(
    Scale Alpha=&FloatFromCurve(FloatCurve=DataInterface() {propertyValues {Curve {keys [
         {arriveTangent=9.0725, interpMode=RCIM_Cubic, leaveTangent=9.0725, tangentMode=RCTM_User, time=0, value=0}
         {arriveTangent=-2.5249, interpMode=RCIM_Cubic, leaveTangent=-2.5249, tangentMode=RCTM_User, time=0.1408, value=0.9847}
         {arriveTangent=-0.0405, interpMode=RCIM_Cubic, leaveTangent=-0.0405, tangentMode=RCTM_User, time=1, value=0}
        ]}}})
    Scale RGB=&VectorFromFloat(Value=&FloatFromCurve(
      FloatCurve=DataInterface() {propertyValues {Curve {keys [
          {arriveTangent=9.0725, interpMode=RCIM_Cubic, leaveTangent=9.0725, tangentMode=RCTM_User, time=0, value=0}
          {arriveTangent=-2.5249, interpMode=RCIM_Cubic, leaveTangent=-2.5249, tangentMode=RCTM_User, time=0.1408, value=0.9847}
          {arriveTangent=-0.0405, interpMode=RCIM_Cubic, leaveTangent=-0.0405, tangentMode=RCTM_User, time=1, value=0}
         ]}}}
      Scale Curve=10
     ))
    ScaleA=true
    ScaleRGB=true
   )
   /SetVariables(Particles.Position=&AddVectorToPosition(
     Position=$Particles.Initial.Position
     Vector=&MakeVector(
      X=0
      Y=0
      Z=&FloatFromCurve(
       FloatCurve=DataInterface() {propertyValues {Curve {keys [{arriveTangent=3.9988, time=0, value=-0.25}, {arriveTangent=0.4807, leaveTangent=0.4807, time=1, value=1.8}]}}}
       Scale Curve=20
      )
     )
    )) {isSetParameters=true}
  ]
  rendererClasses [&NiagaraSpriteRendererProperties]
  renderers [{rendererClass=&NiagaraSpriteRendererProperties}]
 }
 Smoke {
  emitterUpdateScript [
   &EmitterState(Life Cycle Mode=::NewEnumerator1, Loop Behavior=::NewEnumerator1, Loop Duration=0.2, UseLoopDelay=false)
   &SpawnRate(SpawnRate=20, Use Spawn Probability=true) {module=SpawnRate001}
   &SpawnBurst_Instantaneous(Spawn Count=4, Use Spawn Probability=false)
  ]
  particleSpawnScript [
   &InitializeParticle(
    Color Maximum=<rgba 0.714 1 0.4478 0.6>
    Color Minimum=<rgba 0.4121 1 0.7086 0.3>
    Color Mode=::NewEnumerator2
    Lifetime Max=0.5
    Lifetime Min=0.3
    Lifetime Mode=::NewEnumerator1
    Sprite Rotation Mode=::NewEnumerator2
    Sprite Size Mode=::NewEnumerator1
    Uniform Sprite Size Max=50
    Uniform Sprite Size Min=30
    UsePositionOffset=false
    Write Material Random=true
   )
   &SkeletalMeshLocation(Bone / Socket Index=$User.BoneIndex, Bone Sampling Mode=::NewEnumerator4, MaskSpawnGroup=false, Skeletal Mesh=$System.Skeletal Mesh)
  ]
  particleUpdateScript [
   &ParticleState()
   &CameraOffset(Camera Offset Amount=5)
   &GravityForce(Gravity=&MakeVector(
     X=&RandomRangeFloat(Minimum=-1, Maximum=1)
     Y=&RandomRangeFloat(Minimum=-1, Maximum=1)
     Z=&FloatFromCurve(
      FloatCurve=DataInterface() {propertyValues {Curve {keys [
          {arriveTangent=-1.2674, interpMode=RCIM_Cubic, leaveTangent=-1.2674, tangentMode=RCTM_User, time=0, value=1}
          {interpMode=RCIM_Cubic, tangentMode=RCTM_User, time=1, value=0}
         ]}}}
      Scale Curve=&RandomRangeFloat(Minimum=150, Maximum=300)
     )
    ))
   &CurlNoiseForce(Noise Frequency=10, Noise Strength=20, Pan Noise=true, Pan Noise Field=<0 0 1>)
   &Drag(Drag=&RandomRangeFloat(Minimum=1, Maximum=3), Use Linear Drag=true, Use Rotational Drag=false)
   &SolveForcesAndVelocity(Clamp Velocity=true, Limit Acceleration=false, Speed Limit=100)
   &ScaleColor(
    Scale Alpha=&FloatFromCurve(
     FloatCurve=DataInterface() {propertyValues {Curve {keys [
         {arriveTangent=9.0725, interpMode=RCIM_Cubic, leaveTangent=9.0725, tangentMode=RCTM_User, time=0, value=0}
         {arriveTangent=-2.5249, interpMode=RCIM_Cubic, leaveTangent=-2.5249, tangentMode=RCTM_User, time=0.1408, value=0.9847}
         {arriveTangent=-0.0405, interpMode=RCIM_Cubic, leaveTangent=-0.0405, tangentMode=RCTM_User, time=1, value=0}
        ]}}}
     Scale Curve=0.3
    )
    ScaleA=true
    ScaleRGB=true
   )
   &ScaleSpriteSizeBySpeed(Max Scale Factor=<1.1 1.1>, Velocity Threshold=100)
   &ScaleSpriteSize(Uniform Curve Sprite Scale=DataInterface() {propertyValues {Curve {keys [
        {arriveTangent=1.3764, interpMode=RCIM_Cubic, leaveTangent=1.3764, tangentMode=RCTM_User, time=0, value=0.2}
        {interpMode=RCIM_Cubic, tangentMode=RCTM_User, time=1, value=1}
       ]}}})
   &SubUVAnimation()
  ]
  rendererClasses [&NiagaraSpriteRendererProperties]
  renderers [{rendererClass=&NiagaraSpriteRendererProperties}]
 }
 Wave {
  emitterUpdateScript [
   &EmitterState(Life Cycle Mode=::NewEnumerator1, Loop Behavior=::NewEnumerator1, Loop Duration=3, UseLoopDelay=false)
   &SpawnBurst_Instantaneous(Spawn Count=1, Use Spawn Probability=false)
  ]
  particleSpawnScript [
   &InitializeParticle(
    Color=<rgba 1 1 1 0.5>
    Lifetime=&RandomRangeFloat(Minimum=0.35, Maximum=0.4)
    Mesh Scale Mode=::NewEnumerator1
    Mesh Uniform Scale Max=0.4
    Mesh Uniform Scale Min=0.35
    UsePositionOffset=false
    Write Material Random=true
   )
   &CameraOffset()
   &SkeletalMeshLocation(
    Bone / Socket Index=$User.BoneIndex
    Bone Sampling Mode=::NewEnumerator4
    MaskSpawnGroup=false
    Orientation Sampling=::NewEnumerator1
    Skeletal Mesh=$System.Skeletal Mesh
   )
   &InitialMeshOrientation(
    Mesh Axis To Orient=<0 0 1>
    Mesh Orientation Mode=::NewEnumerator2
    Rotation=&MakeVector(X=0, Y=0, Z=&RandomRangeFloat(Minimum=0, Maximum=1))
    Use Rotation Vector=true
   )
   &AddVelocity(Rotation Coordinate Space=::World, Velocity=<0 0 100>)
  ]
  particleUpdateScript [
   &ParticleState()
   &Drag(Drag=3, Use Linear Drag=true, Use Rotational Drag=true)
   &SolveForcesAndVelocity(Clamp Velocity=false, Limit Acceleration=false)
   &ScaleColor(
    Linear Color Curve=DataInterface() {propertyValues {
      AlphaCurve {keys [<at 0.6964 3>,<at 1 0>]}
      BlueCurve {keys [<at 0.5956 0>,<at 1 0.276>]}
      GreenCurve {keys [<at 0.5956 0.2031>,<at 1 0.1523>]}
      RedCurve {keys [<at 0.5956 0.0054>,<at 1 0>]}
     }}
    Scale Mode=::NewEnumerator2
   )
   &ScaleColor(
    Scale Alpha=&FNTimeOfDayScalar(Night Scale=0.7, Morning Scale=0.5, Evening Scale=0.5)
    Scale RGB=&VectorFromFloat(Value=&FloatFromCurve(
      FloatCurve=DataInterface() {propertyValues {Curve {keys [{interpMode=RCIM_Cubic, tangentMode=RCTM_User, time=0, value=1}, {interpMode=RCIM_Cubic, tangentMode=RCTM_User, time=1, value=0}]}}}
      Scale Curve=2
     ))
    ScaleA=true
    ScaleRGB=true
   ) {module=ScaleColor001}
   &DynamicMaterialParameters(
    Index 0 Param 1=&FloatFromCurve(FloatCurve=DataInterface() {propertyValues {Curve {keys [{arriveTangent=-1.5116, interpMode=RCIM_Cubic, leaveTangent=-1.5116, tangentMode=RCTM_User, time=0, value=0.6}, <at 1 0>]}}})
    Index 0 Param 2=&FloatFromCurve(FloatCurve=DataInterface() {propertyValues {Curve {keys [{arriveTangent=-0.5818, interpMode=RCIM_Cubic, leaveTangent=-0.5818, tangentMode=RCTM_User, time=0, value=1}, <at 1 0.9>]}}})
    Index 0 Param 3=&FloatFromCurve(
     FloatCurve=DataInterface() {propertyValues {Curve {keys [
         {arriveTangent=4.1785, interpMode=RCIM_Cubic, leaveTangent=4.1785, tangentMode=RCTM_User, time=0, value=0}
         {arriveTangent=1.1179, interpMode=RCIM_Cubic, leaveTangent=1.1179, tangentMode=RCTM_User, time=0.2042, value=0.6228}
         {arriveTangent=0.0224, interpMode=RCIM_Cubic, leaveTangent=0.0224, tangentMode=RCTM_User, time=1, value=1}
        ]}}}
     Scale Curve=2
    )
    Module.Param0WriteEnabled=true
    Module.Param1WriteEnabled=true
    Module.Param2WriteEnabled=true
    Module.Param3WriteEnabled=false
   )
   &ScaleMeshSize(Scale Factor=&VectorFromFloat(Value=&FloatFromCurve(FloatCurve=DataInterface() {propertyValues {Curve {keys [
          {arriveTangent=1.9365, interpMode=RCIM_Cubic, leaveTangent=1.9365, tangentMode=RCTM_User, time=0, value=0.2}
          {interpMode=RCIM_Cubic, tangentMode=RCTM_User, time=1, value=1}
         ]}}})))
  ]
  rendererClasses [&NiagaraMeshRendererProperties]
  renderers [{rendererClass=&NiagaraMeshRendererProperties, rendererIndex=0}]
 }
}
summary {
 systemName=NS_Emote_CarrotCake_SwitchProp
 emitters [
  {emitterName=Flash_Flare, rendererClasses [&NiagaraSpriteRendererProperties]}
  {emitterName=AirGlow, rendererClasses [&NiagaraSpriteRendererProperties, &NiagaraLightRendererProperties]}
  {emitterName=Embers, rendererClasses [&NiagaraSpriteRendererProperties]}
  {emitterName=Smoke, rendererClasses [&NiagaraSpriteRendererProperties]}
  {emitterName=Wave, rendererClasses [&NiagaraMeshRendererProperties]}
 ]
 userVariables [
  User.bIsFrontEnd {type=NiagaraBool}
  User.bIsFrontEndPreview {type=NiagaraBool}
  User.BoneIndex {type=NiagaraInt32}
  User.RibbonIDOffset {type=NiagaraInt32}
 ]
}
